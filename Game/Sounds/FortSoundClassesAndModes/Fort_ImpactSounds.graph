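# Sound class for impact sounds in the Fortnite sound hierarchy, parented to Fort_08_Player_Interacts.
# It groups child classes for close, distant and duck-named impact variants, plus player construction and construction UI sounds.
# It lowers base volume, turns reverb off, applies ambient volumes and defaults to the Fortnite_SFX submix.
# Volume modulation is enabled through the FortImpacts and FortSFX volume patches.
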SoundClass /Game/Sounds/FortSoundClassesAndModes/Fort_ImpactSounds
parent=&/Fort_08_Player_Interacts
children [
 &/Fort_ImpactSounds_Close
 &/Fort_ImpactSounds_Default_Close
 &/Fort_ImpactSounds_Distant
 &/Fort_ImpactSounds_DuckDefault_PSM
 &/Fort_ImpactSounds_DuckDefaultQuieter_PSM
 &/Fort_Player_Construction
 &/Fort_Player_Construction_UI
]
properties {properties {
  bApplyAmbientVolumes=true
  bReverb=false
  defaultSubmix=&Fortnite_SFX
  volume=0.6
  modulationSettings {volumeModulationDestination {
    bEnableModulation=true
    value=1
    modulators [/Game/Sounds/Modulation/Patches/Patch_FortImpacts_Vol, /Game/Sounds/Modulation/Patches/Patch_FortSFX_Vol]
   }}
 }}
