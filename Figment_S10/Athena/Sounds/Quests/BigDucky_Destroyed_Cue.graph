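# Sound cue for the Big Ducky destroyed sound, stored with quest audio in the Figment Season 10 content.
# A Random node picks between two destruction wave variants with equal weight and feeds a Modulator at the root for per-play variation.
# It plays at reduced volume through the impact sound class, spatialized with the gadget-activate attenuation settings.
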SoundCue /Figment_S10/Athena/Sounds/Quests/BigDucky_Destroyed_Cue
root=Modulator_1
nodes [
 {id=Modulator_1, children [Random_1]}
 {id=Random_1, children [WavePlayer_1,WavePlayer_2], properties {weights [1,1]}}
 {id=WavePlayer_1, properties {soundWaveAssetPtr=&/BigDucky_Destroyed_02}}
 {id=WavePlayer_2, properties {soundWaveAssetPtr=&/BigDucky_Destroyed_01}}
]
properties {
 attenuationSettings=&Fort_GadgetActivate_Att
 maxDistance=4000
 soundClassObject=/Game/Sounds/FortSoundClassesAndModes/Fort_ImpactSounds
 volumeMultiplier=0.65
}
