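# BigDucky_Quack_Cue is a Sound Cue that plays a quack for BigDucky, filed under Athena quest sounds.
# A Random node picks one of four equally weighted quack waves and feeds it through a Modulator at the root.
# It plays at reduced volume in a projectile weapon sound class, with overridden attenuation adding distance falloff, stereo spread, listener focus and occlusion filtering.
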SoundCue /Figment_S10/Athena/Sounds/Quests/BigDucky_Quack_Cue
root=Modulator_0
nodes [
 {id=Modulator_0, children [Random_1]}
 {id=Random_1, children [WavePlayer_1,WavePlayer_5,WavePlayer_6,WavePlayer_7], properties {weights [1,1,1,1]}}
 {id=WavePlayer_1, properties {soundWaveAssetPtr=/Game/Athena/Sounds/Quests/BigDucky_Quack_01}}
 {id=WavePlayer_5, properties {soundWaveAssetPtr=/Game/Athena/Sounds/Quests/BigDucky_Quack_02}}
 {id=WavePlayer_6, properties {soundWaveAssetPtr=/Game/Athena/Sounds/Quests/BigDucky_Quack_03}}
 {id=WavePlayer_7, properties {soundWaveAssetPtr=&/BigDucky_Quack_04}}
]
properties {
 attenuationSettings=&Fort_GadgetActivate_Att
 bOverrideAttenuation=true
 maxDistance=6000
 soundClassObject=/Figment_S10/Sounds/FortSoundClassesAndModes/Fort_Weapon_Projectile_PSM
 volumeMultiplier=0.5
 attenuationOverrides {
  bEnableListenerFocus=true
  bEnableOcclusion=true
  falloffDistance=5000
  nonFocusVolumeAttenuation=0.3
  occlusionLowPassFilterFrequency=10000
  occlusionVolumeAttenuation=0.5
  stereoSpread=500
  attenuationShapeExtents {x=1000}
 }
}
