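# Looping ambient sound cue, named for a quest storm center, that reuses a shield power loop at reduced volume and raised base pitch.
# An endlessly looping Enveloper sweeps a pitch curve up and back down over ten seconds, with volume held flat.
# A quality-level node feeds it the standard wave or an optimized variant.
# Attenuation overrides localize the sound, which plays in the gadgets sound class.
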SoundCue /Figment_S10/Athena/Sounds/Quests/StormCenter_AmbientLoop_Cue
root=Enveloper_0
nodes [
 {
  id=Enveloper_0
  children [QualityLevel_1]
  properties {
   bLoop=true
   bLoopIndefinitely=true
   pitchCurve {editorCurveData {keys [{time=0}, <at 5 0.75>, {time=10}]}}
   volumeCurve {editorCurveData {keys [<at 0 1>]}}
  }
 }
 {id=QualityLevel_1, children [WavePlayer_2,WavePlayer_2,WavePlayer_3,WavePlayer_3]}
 {id=WavePlayer_2, properties {bLooping=true, soundWaveAssetPtr=/Game/Athena/Sounds/Shield/Shield_PowerLoop_02}}
 {id=WavePlayer_3, properties {bLooping=true, soundWaveAssetPtr=/Game/Athena/Sounds/Shield/Shield_PowerLoop_02_opt}}
]
properties {
 bOverrideAttenuation=true
 maxDistance=2560
 pitchMultiplier=1.5
 soundClassObject=&Fort_Gadgets
 volumeMultiplier=0.25
 attenuationOverrides {falloffDistance=1536, attenuationShapeExtents {x=1024}}
}
