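# StormCenter_Collect_Cue is a quest sound cue, apparently for collecting a storm center, that plays a distant thunder sample.
# A quality-level node selects the branch: the first two levels play Dist_Thunder_001 unmodified, while the last two play Dist_Thunder_002_opt through a pitch-randomizing modulator.
# The cue is pitched up overall, played at reduced volume and routed to the Fort_UI sound class.
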SoundCue /Figment_S10/Athena/Sounds/Quests/StormCenter_Collect_Cue
root=QualityLevel_0
nodes [
 {id=Modulator_0, children [WavePlayer_22], properties {pitchMax=1.1, pitchMin=0.8}}
 {id=QualityLevel_0, children [WavePlayer_0,WavePlayer_0,Modulator_0,Modulator_0]}
 {id=WavePlayer_0, properties {soundWaveAssetPtr=/Game/Sounds/Fort_GamePlay_Sounds/Storms/Dist_Thunder_001}}
 {id=WavePlayer_22, properties {soundWaveAssetPtr=/Game/Sounds/Fort_GamePlay_Sounds/Storms/Dist_Thunder_002_opt}}
]
properties {maxDistance=2097152, pitchMultiplier=1.25, soundClassObject=&Fort_UI, volumeMultiplier=0.2}
